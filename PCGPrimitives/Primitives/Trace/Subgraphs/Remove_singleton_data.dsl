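; PCG subgraph in the Trace primitives, named for removing singleton data, that filters its input by element count.
; Get Element Count feeds a NotEqual Attribute Compare Op, whose result drives a Branch on the same input.
; Data whose count differs from the comparison value goes to the unconnected B side, so only Branch Output A reaches the graph output.
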(graph Remove_singleton_data
  (bind GetElementsCount_0 (Get%20Element%20Count))
  (bind AttributeCompareOp_1 (Attribute%20Compare%20Op :Operation NotEqual))
  (bind Branch_0 (Branch))
  (connect GetElementsCount_0 Out AttributeCompareOp_1 InA)
  (connect AttributeCompareOp_1 Out Branch_0 bOutputToB)
  (connect DefaultInputNode In Branch_0 In)
  (connect DefaultInputNode In GetElementsCount_0 In)
  (output Out Branch_0.Output%20A))
