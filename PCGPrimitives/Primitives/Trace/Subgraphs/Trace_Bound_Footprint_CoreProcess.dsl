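; PCG subgraph that traces the rectangular footprint of the combined bounds of input points, splines and 2D polygons on a selectable planar surface.
; Inputs become points merged into one bounds point, which is offset and scaled.
; Four point-center resets, driven by the PlanarSurface row of an offsets table, yield corners that receive PositionOffset again.
; AsPoint selects either one combined corner point or a closed linear counter-clockwise spline.
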(graph Trace_Bound_Footprint_CoreProcess
  (params (PlanarSurface :type (Enum /PCGPrimitives/Primitives/Trace/Resources/EPCG_PlanarSurfacesEnum.EPCG_PlanarSurfacesEnum) :default NewEnumerator5) (PositionOffset :type Vector) (ScaleFactor :type Vector :default (1.0 1.0 1.0)) (AsPoint :type Bool))
  (bind FilterDataByType_0 (Filter%20Data%20By%20Type :TargetType (struct :Ids [(struct :Struct /Script/PCG.PCGDataTypeInfoPoint)])))
  (bind FilterDataByType_1 (Filter%20Data%20By%20Type :TargetType (struct :Ids [(struct :Struct /Script/PCG.PCGDataTypeInfoSpline)])))
  (bind FilterDataByType_2 (Filter%20Data%20By%20Type :TargetType (struct :Ids [(struct :Struct /Script/PCG.PCGDataTypeInfoPolygon2D)])))
  (bind GetSplineControlPoints_0 (Get%20Spline%20Control%20Points))
  (bind MergePoints_1 (Merge%20Points))
  (bind CombinePoints_0 (Combine%20Points))
  (bind AttributeMathsOp_2 (Attribute%20Maths%20Op :InputSource1 $Position :InB PositionOffset))
  (bind AttributeMathsOp_1 (Attribute%20Maths%20Op :InputSource1 $Scale :Operation Multiply :InB ScaleFactor))
  (bind ApplyScaleToBounds_0 (Apply%20Scale%20To%20Bounds))
  (bind PCGLoadDataAsset_0 (Load%20PCG%20Data%20Asset :Asset /PCGPrimitives/Primitives/Trace/Resources/PlanarSurfacesPointOffsetsTable.PlanarSurfacesPointOffsetsTable :bTagOutputsBasedOnOutputPins false :Pins [(struct :Label Out :AllowedTypes (struct :Ids [(struct :Struct /Script/PCG.PCGDataTypeInfoParam)]))]))
  (bind FilterElementsByIndex_0 (Filter%20Elements%20By%20Index :Filter PlanarSurface))
  (bind ExtractAttribute_0 (Extract%20Attribute%20at%20Index :bForceOutputAttributeToBeInElementsDomain true :InputSource A))
  (bind ResetPointCenter_0 (Reset%20Point%20Center :PointCenterLocation (1.0 1.0 0.5)))
  (bind ExtractAttribute_1 (Extract%20Attribute%20at%20Index :bForceOutputAttributeToBeInElementsDomain true :InputSource B))
  (bind ResetPointCenter_1 (Reset%20Point%20Center :PointCenterLocation (0.0 1.0 0.5)))
  (bind ExtractAttribute_2 (Extract%20Attribute%20at%20Index :bForceOutputAttributeToBeInElementsDomain true :InputSource C))
  (bind ResetPointCenter_2 (Reset%20Point%20Center :PointCenterLocation (0.0 0.0 0.5)))
  (bind ExtractAttribute_3 (Extract%20Attribute%20at%20Index :bForceOutputAttributeToBeInElementsDomain true :InputSource D))
  (bind ResetPointCenter_3 (Reset%20Point%20Center :PointCenterLocation (1.0 0.0 0.5)))
  (bind MergePoints_0 (Merge%20Points))
  (bind AttributeMathsOp_0 (Attribute%20Maths%20Op :InputSource1 $Position :InB PositionOffset))
  (bind CreateSpline_0 (Create%20Spline :bClosedLoop true :bLinear true :InterpTypeAttribute None))
  (bind CombinePoints_1 (Combine%20Points))
  (bind SplineDirection_0 (Spline%20Direction :Operation ForceCounterClockwise))
  (bind Select_0 (Select :bUseInputB-param AsPoint))
  (connect AttributeMathsOp_1 Out ApplyScaleToBounds_0 In)
  (connect MergePoints_0 Out AttributeMathsOp_0 InA)
  (connect AttributeMathsOp_2 Out AttributeMathsOp_1 InA)
  (connect CombinePoints_0 Out AttributeMathsOp_2 InA)
  (connect MergePoints_1 Out CombinePoints_0 In)
  (connect AttributeMathsOp_0 Out CombinePoints_1 In)
  (connect AttributeMathsOp_0 Out CreateSpline_0 In)
  (connect FilterElementsByIndex_0 In%20Filter ExtractAttribute_0 In)
  (connect FilterElementsByIndex_0 In%20Filter ExtractAttribute_1 In)
  (connect FilterElementsByIndex_0 In%20Filter ExtractAttribute_2 In)
  (connect FilterElementsByIndex_0 In%20Filter ExtractAttribute_3 In)
  (connect DefaultInputNode In FilterDataByType_0 In)
  (connect DefaultInputNode In FilterDataByType_1 In)
  (connect DefaultInputNode In FilterDataByType_2 In)
  (connect PCGLoadDataAsset_0 Out FilterElementsByIndex_0 In)
  (connect FilterDataByType_1 InsideFilter GetSplineControlPoints_0 In)
  (connect FilterDataByType_2 InsideFilter GetSplineControlPoints_0 In)
  (connect ResetPointCenter_0 Out MergePoints_0 In)
  (connect ResetPointCenter_1 Out MergePoints_0 In)
  (connect ResetPointCenter_2 Out MergePoints_0 In)
  (connect ResetPointCenter_3 Out MergePoints_0 In)
  (connect FilterDataByType_0 InsideFilter MergePoints_1 In)
  (connect GetSplineControlPoints_0 Out MergePoints_1 In)
  (connect ApplyScaleToBounds_0 Out ResetPointCenter_0 In)
  (connect ExtractAttribute_0 Out ResetPointCenter_0 PointCenterLocation)
  (connect ApplyScaleToBounds_0 Out ResetPointCenter_1 In)
  (connect ExtractAttribute_1 Out ResetPointCenter_1 PointCenterLocation)
  (connect ApplyScaleToBounds_0 Out ResetPointCenter_2 In)
  (connect ExtractAttribute_2 Out ResetPointCenter_2 PointCenterLocation)
  (connect ApplyScaleToBounds_0 Out ResetPointCenter_3 In)
  (connect ExtractAttribute_3 Out ResetPointCenter_3 PointCenterLocation)
  (connect SplineDirection_0 Out Select_0 Input%20A)
  (connect CombinePoints_1 Out Select_0 Input%20B)
  (connect CreateSpline_0 Out SplineDirection_0 In)
  (output Out Select_0))
