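; PCG subgraph that builds bridge splines between two point sets by pairing points of the same index.
; The "Input:0" tag separates primary from secondary input, and each side is filtered to points and merged.
; Branches pass both sets on only when their element counts are equal.
; Each set is split per point and stamped with data_index, then merged and repartitioned by that attribute, so each pair becomes one spline.
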(graph Trace_Bridges_CoreProcess
  (bind FilterDataByType_6 (Filter%20Data%20By%20Type :TargetType (struct :Ids [(struct :Struct /Script/PCG.PCGDataTypeInfoPoint)])))
  (bind MergePoints_0 (Merge%20Points))
  (bind FilterDataByType_7 (Filter%20Data%20By%20Type :TargetType (struct :Ids [(struct :Struct /Script/PCG.PCGDataTypeInfoPoint)])))
  (bind MergePoints_1 (Merge%20Points))
  (bind AttributePartition_0 (Attribute%20Partition :PartitionAttributeSelectors [$Index]))
  (bind AttributePartition_1 (Attribute%20Partition :PartitionAttributeSelectors [$Index]))
  (bind WriteDataIndex_2 (Write%20Data%20Index :bAddAttribute true :bAddTag false :IndexAttribute data_index))
  (bind WriteDataIndex_3 (Write%20Data%20Index :bAddAttribute true :bAddTag false :IndexAttribute data_index))
  (bind MergePoints_2 (Merge%20Points :MergeTagsMode None))
  (bind AttributePartition_2 (Attribute%20Partition :PartitionAttributeSelectors [data_index]))
  (bind CreateSpline_0 (Create%20Spline :InterpTypeAttribute None))
  ;[0 0 0 0.5 12] Primary/Secondary Inputs
  (bind FilterDataByTag_0 (Filter%20Data%20By%20Tag :SelectedTags "Input:0"))
  ;[0.15 0.15 0.15 0.5] EQUAL NUMBER ? if yes > go
  (bind GetElementsCount_0 (Get%20Element%20Count))
  (bind GetElementsCount_1 (Get%20Element%20Count))
  (bind AttributeCompareOp_0 (Attribute%20Compare%20Op))
  (bind Branch_0 (Branch))
  (bind Branch_1 (Branch))
  (connect GetElementsCount_0 Out AttributeCompareOp_0 InA)
  (connect GetElementsCount_1 Out AttributeCompareOp_0 InB)
  (connect Branch_0 Output%20B AttributePartition_0 In)
  (connect Branch_1 Output%20B AttributePartition_1 In)
  (connect MergePoints_2 Out AttributePartition_2 In)
  (connect AttributeCompareOp_0 Out Branch_0 bOutputToB)
  (connect MergePoints_0 Out Branch_0 In)
  (connect AttributeCompareOp_0 Out Branch_1 bOutputToB)
  (connect MergePoints_1 Out Branch_1 In)
  (connect AttributePartition_2 Out CreateSpline_0 In)
  (connect DefaultInputNode In FilterDataByTag_0 In)
  (connect FilterDataByTag_0 InsideFilter FilterDataByType_6 In)
  (connect FilterDataByTag_0 OutsideFilter FilterDataByType_7 In)
  (connect MergePoints_0 Out GetElementsCount_0 In)
  (connect MergePoints_1 Out GetElementsCount_1 In)
  (connect FilterDataByType_6 InsideFilter MergePoints_0 In)
  (connect FilterDataByType_7 InsideFilter MergePoints_1 In)
  (connect WriteDataIndex_2 Out MergePoints_2 In)
  (connect WriteDataIndex_3 Out MergePoints_2 In)
  (connect AttributePartition_0 Out WriteDataIndex_2 In)
  (connect AttributePartition_1 Out WriteDataIndex_3 In)
  (output Out CreateSpline_0))
