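; PCG core subgraph for Trace Connect that turns incoming point data into a spline.
; Per_Points selects either a partition by $Index, giving each point its own data, or Combine Points, collapsing each data to one point.
; The selected data get a dataIndex attribute, then merge into one point set that feeds Create Spline.
; Linear_spline toggles linear spline segments.
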(graph Trace_Connect_CoreProcess
  (params (Per_Points :type Bool) (Linear_spline :type Bool))
  (bind FilterDataByType_7 (Filter%20Data%20By%20Type :TargetType (struct :Ids [(struct :Struct /Script/PCG.PCGDataTypeInfoPoint)])))
  (bind AttributePartition_1 (Attribute%20Partition :PartitionAttributeSelectors [$Index]))
  (bind CombinePoints_0 (Combine%20Points))
  (bind Select_0 (Select :bUseInputB-param Per_Points))
  (bind WriteDataIndex_0 (Write%20Data%20Index :bAddAttribute true :IndexAttribute dataIndex))
  (bind MergePoints_2 (Merge%20Points))
  (bind CreateSpline_0 (Create%20Spline :InterpTypeAttribute None :bLinear-param Linear_spline))
  (connect FilterDataByType_7 InsideFilter AttributePartition_1 In)
  (connect FilterDataByType_7 InsideFilter CombinePoints_0 In)
  (connect MergePoints_2 Out CreateSpline_0 In)
  (connect DefaultInputNode In FilterDataByType_7 In)
  (connect WriteDataIndex_0 Out MergePoints_2 In)
  (connect CombinePoints_0 Out Select_0 Input%20A)
  (connect AttributePartition_1 Out Select_0 Input%20B)
  (connect Select_0 Out WriteDataIndex_0 In)
  (output Out CreateSpline_0))
