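; PCG subgraph that lofts input splines into a parametric grid of Latitude and Longitude tagged splines.
; It proceeds only when all splines share control point count and open/closed state.
; Control points with matching indices are joined in input order into curved or linear Longitude splines.
; These are sampled per Transition_Splines, then regrouped by sample index into Latitude splines, closed when the inputs are closed.
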(graph Trace_Splines_ParametricGrid_CorePocess
  (params (Transition_Splines :type Int :default 5) (Linear_Interpolation :type Bool))
  (bind DeleteTags_0 (Delete%20Tags :Operator Substring :SelectedTags "Type"))
  (bind FilterDataByType_2 (Filter%20Data%20By%20Type :TargetType (struct :Ids [(struct :Struct /Script/PCG.PCGDataTypeInfoSpline)])))
  (bind GetSplineControlPoints_1 (Get%20Spline%20Control%20Points))
  (bind CopyAttributes_1 (Copy%20Attributes :InputSource $Index :OutputTarget origIndex))
  (bind WriteDataIndex_2 (Write%20Data%20Index :bAddAttribute true :bAddTag false :IndexAttribute InSplineIndex :IndexTag ""))
  (bind MergePoints_1 (Merge%20Points))
  (bind AttributePartition_1 (Attribute%20Partition :PartitionAttributeSelectors [origIndex]))
  (bind SortAttributes_0 (Sort%20Attributes :bUseStableSort false :SortAttributes [(struct :InputSource InSplineIndex)]))
  (bind Branch_5 (Branch :bOutputToB-param Linear_Interpolation))
  (bind CreateSpline_1 (Create%20Spline :InterpTypeAttribute None))
  (bind CreateSpline_4 (Create%20Spline :bLinear true :InterpTypeAttribute None))
  (bind SplineSampler_0 (Spline%20Sampler :SamplerParams (struct :SubdivisionsPerSegment 5 :NumSamples 4 :bUnbounded true :PointSteepness 1.0) :NumSamples Transition_Splines :SubdivisionsPerSegment Transition_Splines))
  (bind CopyAttributes_0 (Copy%20Attributes :InputSource $Index :OutputTarget Transition_Spline_index))
  (bind MergePoints_4 (Merge%20Points))
  (bind AttributePartition_4 (Attribute%20Partition :PartitionAttributeSelectors [Transition_Spline_index]))
  (bind Branch_4 (Branch))
  (bind CreateSpline_2 (Create%20Spline :InterpTypeAttribute None))
  (bind CreateSpline_3 (Create%20Spline :bClosedLoop true :InterpTypeAttribute None))
  (bind AddTags_2 (Add%20Tags :TagsToAdd "Type:ClosedSpline"))
  (bind AddTags_3 (Add%20Tags :TagsToAdd "Type:Spline"))
  (bind AddTags_0 (Add%20Tags :TagsToAdd "Latitude"))
  (bind AddTags_4 (Add%20Tags :TagsToAdd "Type:Spline"))
  (bind AddTags_1 (Add%20Tags :TagsToAdd "Longitude"))
  (bind ExtractAttribute_1 (Extract%20Attribute%20at%20Index :bForceOutputAttributeToBeInElementsDomain true))
  ;[0.15 0.15 0.15 0.5] ALL OPEN or CLOSED ? if YES > go
  (bind ExtractAttribute_0 (Extract%20Attribute%20at%20Index :bForceOutputAttributeToBeInElementsDomain true :InputSource @data.$IsClosed))
  (bind MergeAttributes_0 (Merge%20Attributes))
  (bind AttributeCast_0 (Attribute%20Cast :OutputType Integer32))
  (bind AttributeReduce_2 (Attribute%20Reduce :Operation Max))
  (bind AttributeReduce_3 (Attribute%20Reduce :Operation Min))
  (bind AttributeCompareOp_3 (Attribute%20Compare%20Op))
  (bind Branch_3 (Branch))
  ;[0.15 0.15 0.15 0.5] EQUAL NUMBER of control points? if yes > go
  (bind GetElementsCount_3 (Get%20Element%20Count))
  (bind AttributeReduce_0 (Attribute%20Reduce :Operation Max))
  (bind AttributeReduce_1 (Attribute%20Reduce :Operation Min))
  (bind AttributeCompareOp_2 (Attribute%20Compare%20Op))
  (bind Branch_1 (Branch))
  (reroute IsCLosed ExtractAttribute_1)
  (connect AddTags_2 Out AddTags_0 In)
  (connect AddTags_3 Out AddTags_0 In)
  (connect AddTags_4 Out AddTags_1 In)
  (connect CreateSpline_3 Out AddTags_2 In)
  (connect CreateSpline_2 Out AddTags_3 In)
  (connect CreateSpline_1 Out AddTags_4 In)
  (connect CreateSpline_4 Out AddTags_4 In)
  (connect MergeAttributes_0 Out AttributeCast_0 In)
  (connect AttributeReduce_0 Out AttributeCompareOp_2 InA)
  (connect AttributeReduce_1 Out AttributeCompareOp_2 InB)
  (connect AttributeReduce_2 Out AttributeCompareOp_3 InA)
  (connect AttributeReduce_3 Out AttributeCompareOp_3 InB)
  (connect Branch_3 Output%20B AttributePartition_1 In)
  (connect MergePoints_4 Out AttributePartition_4 In)
  (connect GetElementsCount_3 Out AttributeReduce_0 In)
  (connect GetElementsCount_3 Out AttributeReduce_1 In)
  (connect AttributeCast_0 Out AttributeReduce_2 In)
  (connect AttributeCast_0 Out AttributeReduce_3 In)
  (connect AttributeCompareOp_2 Out Branch_1 bOutputToB)
  (connect MergePoints_1 Out Branch_1 In)
  (connect AttributeCompareOp_3 Out Branch_3 bOutputToB)
  (connect Branch_1 Output%20B Branch_3 In)
  (connect IsCLosed Out Branch_4 bOutputToB)
  (connect AttributePartition_4 Out Branch_4 In)
  (connect SortAttributes_0 Out Branch_5 In)
  (connect SplineSampler_0 Out CopyAttributes_0 Source)
  (connect SplineSampler_0 Out CopyAttributes_0 Target)
  (connect GetSplineControlPoints_1 Out CopyAttributes_1 Source)
  (connect GetSplineControlPoints_1 Out CopyAttributes_1 Target)
  (connect Branch_5 Output%20A CreateSpline_1 In)
  (connect Branch_4 Output%20A CreateSpline_2 In)
  (connect Branch_4 Output%20B CreateSpline_3 In)
  (connect Branch_5 Output%20B CreateSpline_4 In)
  (connect DefaultInputNode In DeleteTags_0 In)
  (connect FilterDataByType_2 InsideFilter ExtractAttribute_0 In)
  (connect MergeAttributes_0 Out ExtractAttribute_1 In)
  (connect DeleteTags_0 Out FilterDataByType_2 In)
  (connect FilterDataByType_2 InsideFilter GetElementsCount_3 In)
  (connect FilterDataByType_2 InsideFilter GetSplineControlPoints_1 In)
  (connect ExtractAttribute_0 Out MergeAttributes_0 In)
  (connect WriteDataIndex_2 Out MergePoints_1 In)
  (connect CopyAttributes_0 Out MergePoints_4 In)
  (connect AttributePartition_1 Out SortAttributes_0 In)
  (connect CreateSpline_1 Out SplineSampler_0 Spline)
  (connect CreateSpline_4 Out SplineSampler_0 Spline)
  (connect CopyAttributes_1 Out WriteDataIndex_2 In)
  (output Out AddTags_0)
  (output Out AddTags_1))
